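# Physics asset for the dynamic right-side collar on the M_MED_Looper male medium player body.
# Capsules on the back, side and front collar bones are each constrained to a small _MSTR anchor body, with tight twist and swing1 limits and a wider swing2 that varies per bone.
# Head, neck and right upper-arm twist capsules supply surrounding body shapes, with the head-to-neck constraint left free in swing and twist.
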PhysicsAsset /BRCosmetics/Characters/Player/Male/Medium/Bodies/M_MED_Looper/Meshes/M_MED_Looper_Dyn_Collar
bodies {
 dyn_collar_bk_r [{shapeType=Capsule, length=6.2634, radius=5.2066, shapeName=dyn_jacket_l_fr_01_capsule, center=<3 0 3>, rotation=<rot 0 0 90>}]
 dyn_collar_bk_r_MSTR [{shapeType=Capsule, length=1, radius=1}]
 dyn_collar_ft_r [{shapeType=Capsule, length=5, radius=5, shapeName=dyn_jacket_l_fr_01_capsule, center=<2 0 4>, rotation=<rot 0 0 90>}]
 dyn_collar_ft_r_MSTR [{shapeType=Capsule, length=1, radius=1}]
 dyn_collar_r [{shapeType=Capsule, length=6.2634, radius=5.2066, shapeName=dyn_jacket_l_fr_01_capsule, center=<3 0 3>, rotation=<rot 0 -20.0001 90.0001>}]
 dyn_collar_r_MSTR [{shapeType=Capsule, length=1, radius=1}]
 head [{shapeType=Capsule, length=5.2437, radius=6.9334, center=<1.5767 3.6196 0>, rotation=<rot 90 0 15.0001>}]
 neck_02 [{shapeType=Capsule, length=1.7223, radius=5.439, center=<3.1394 0 0>, rotation=<rot 90 0 0>}]
 upperarm_twist_01_r [{
   shapeType=Capsule
   length=7.9682
   radius=8.5602
   shapeName=upperarm_twist_01_r_capsule
   center=<-8.5718 11.3006 4.1785>
   rotation=<rot 90 0 0>
  }]
}
constraints [
 {bone1Name=dyn_collar_bk_r, bone2Name=dyn_collar_bk_r_MSTR, swing1LimitDegrees=2, swing2LimitDegrees=10, twistLimitDegrees=2}
 {bone1Name=dyn_collar_r, bone2Name=dyn_collar_r_MSTR, swing1LimitDegrees=2, swing2LimitDegrees=25, twistLimitDegrees=2}
 {bone1Name=dyn_collar_ft_r, bone2Name=dyn_collar_ft_r_MSTR, swing1LimitDegrees=2, swing2LimitDegrees=15, twistLimitDegrees=2}
 {
  bone1Name=head
  bone2Name=neck_02
  swing1LimitDegrees=45
  swing1Motion=Free
  swing2LimitDegrees=45
  swing2Motion=Free
  twistLimitDegrees=45
  twistMotion=Free
 }
]
