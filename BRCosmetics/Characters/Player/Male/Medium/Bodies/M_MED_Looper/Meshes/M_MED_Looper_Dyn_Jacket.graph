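# Physics asset for the M_MED_Looper player body's dynamic jacket, right-side collar and compass bones, using capsule bodies.
# Seven two-segment jacket chains and three collar flaps hang with tight angular limits from small MSTR anchors constrained to spine_01.
# Cross-constraints with mostly free rotation link some neighboring lower jacket segments.
# Pelvis, spine, thigh, neck, head and right upper-arm capsules supply body shapes, with the compass constrained to the pelvis.
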PhysicsAsset /BRCosmetics/Characters/Player/Male/Medium/Bodies/M_MED_Looper/Meshes/M_MED_Looper_Dyn_Jacket
bodies {
 dyn_collar_bk_r [{shapeType=Capsule, length=6.2634, radius=5.2066, shapeName=dyn_jacket_l_fr_01_capsule, center=<3 0 3>, rotation=<rot 0 0 90>}]
 dyn_collar_bk_r_MSTR [{shapeType=Capsule, length=1, radius=1}]
 dyn_collar_ft_r [{shapeType=Capsule, length=5, radius=5, shapeName=dyn_jacket_l_fr_01_capsule, center=<2 0 4>, rotation=<rot 0 0 90>}]
 dyn_collar_ft_r_MSTR [{shapeType=Capsule, length=1, radius=1}]
 dyn_collar_r [{shapeType=Capsule, length=6.2634, radius=5.2066, shapeName=dyn_jacket_l_fr_01_capsule, center=<3 0 3>, rotation=<rot 0 -20.0001 90.0001>}]
 dyn_collar_r_MSTR [{shapeType=Capsule, length=1, radius=1}]
 dyn_compass [{shapeType=Capsule, length=4.2177, radius=8.3737, center=<7.1919 -0.368 -3.9533>, rotation=<rot 90 0 0>}]
 dyn_jacket_bk_01 [{shapeType=Capsule, length=14.317, radius=7.6129, shapeName=dyn_jacket_l_bk_01_capsule, center=<6 0 6>, rotation=<rot 0 0 90>}]
 dyn_jacket_bk_02 [{shapeType=Capsule, length=14.5266, radius=7.4033, shapeName=dyn_jacket_l_bk_02_capsule, center=<6 0 6>, rotation=<rot 0 0 90>}]
 dyn_jacket_bk_MSTR [{shapeType=Capsule, length=1, radius=1}]
 dyn_jacket_l_01 [{shapeType=Capsule, length=14.3188, radius=6.3717, shapeName=dyn_jacket_l_01_capsule, center=<6 0 7>, rotation=<rot 0 0 90>}]
 dyn_jacket_l_02 [{shapeType=Capsule, length=15.585, radius=7.4516, shapeName=dyn_jacket_l_02_capsule, center=<6 0 7>, rotation=<rot 0 0 90>}]
 dyn_jacket_l_MSTR [{shapeType=Capsule, length=1, radius=1}]
 dyn_jacket_l_bk_01 [{shapeType=Capsule, length=17.2673, radius=6.3241, shapeName=dyn_jacket_l_bk_01_capsule, center=<6 0 7>, rotation=<rot -0.0001 0.0001 95>}]
 dyn_jacket_l_bk_02 [{shapeType=Capsule, length=15.9251, radius=7.6663, shapeName=dyn_jacket_l_bk_02_capsule, center=<6 0 7>, rotation=<rot -0.0001 0.0001 95>}]
 dyn_jacket_l_bk_MSTR [{shapeType=Capsule, length=1, radius=1}]
 dyn_jacket_l_fr_01 [{shapeType=Capsule, length=4.1912, radius=7.4601, shapeName=dyn_jacket_l_fr_01_capsule, center=<6 0 7>, rotation=<rot 90 180 180>}]
 dyn_jacket_l_fr_02 [{shapeType=Capsule, length=8.4398, radius=7.4601, shapeName=dyn_jacket_l_fr_02_capsule, center=<6 -1 8>, rotation=<rot 0 0 105>}]
 dyn_jacket_l_fr_MSTR [{shapeType=Capsule, length=1, radius=1}]
 dyn_jacket_r_01 [{shapeType=Capsule, length=15.8142, radius=7.4704, shapeName=dyn_jacket_l_01_capsule, center=<6 0 7>, rotation=<rot 0 0 90>}]
 dyn_jacket_r_02 [{shapeType=Capsule, length=16.2941, radius=6.9905, shapeName=dyn_jacket_l_02_capsule, center=<6 0 7>, rotation=<rot 0 0 90>}]
 dyn_jacket_r_MSTR [{shapeType=Capsule, length=1, radius=1}]
 dyn_jacket_r_bk_01 [{shapeType=Capsule, length=12.5847, radius=6.4379, shapeName=dyn_jacket_l_bk_01_capsule, center=<6 0 7>, rotation=<rot -0.0001 -0.0001 85>}]
 dyn_jacket_r_bk_02 [{shapeType=Capsule, length=12.6689, radius=6.3537, shapeName=dyn_jacket_l_bk_02_capsule, center=<6 0 7>, rotation=<rot -0.0002 -0.0001 90>}]
 dyn_jacket_r_bk_MSTR [{shapeType=Capsule, length=1, radius=1}]
 dyn_jacket_r_fr_01 [{shapeType=Capsule, length=4.1912, radius=7.4601, shapeName=dyn_jacket_l_fr_01_capsule, center=<6 0 7>, rotation=<rot 90 180 180>}]
 dyn_jacket_r_fr_02 [{shapeType=Capsule, length=7.5151, radius=7.4601, shapeName=dyn_jacket_l_fr_02_capsule, center=<6 2 8>, rotation=<rot 0.0001 -0.0002 80>}]
 dyn_jacket_r_fr_MSTR [{shapeType=Capsule, length=1, radius=1}]
 head [{shapeType=Capsule, length=5.2437, radius=6.9334, center=<1.5767 3.6196 0>, rotation=<rot 90 0 15.0001>}]
 neck_02 [{shapeType=Capsule, length=1.7223, radius=5.439, center=<3.1394 0 0>, rotation=<rot 90 0 0>}]
 pelvis [{shapeType=Capsule, length=7.5362, radius=12.1431, center=<-0.9769 -0.7557 0.2377>}]
 spine_01 [{shapeType=Capsule, length=8.1566, radius=11.4829, center=<-1.072 -0.8263 0>}]
 thigh_l [{shapeType=Capsule, length=29.809, radius=10.203, center=<-14.0598 -1.0321 1.4501>, rotation=<rot 88.0959 3.8383 -174.6616>}]
 thigh_twist_01_l [{shapeType=Capsule, length=14.9124, radius=9.9916, center=<4.8189 0.0398 -1.5772>, rotation=<rot 85.4073 3.5381 -178.1736>}]
 thigh_twist_01_r [{shapeType=Capsule, length=24.6494, radius=8.0826, center=<-7.9189 0.5362 -0.3869>, rotation=<rot 88.768 7.5181 5.4337>}]
 upperarm_twist_01_r [{
   shapeType=Capsule
   length=7.9682
   radius=8.5602
   shapeName=upperarm_twist_01_r_capsule
   center=<-8.5718 11.3006 4.1785>
   rotation=<rot 90 0 0>
  }]
}
constraints [
 {bone1Name=dyn_jacket_r_bk_01, bone2Name=dyn_jacket_r_bk_MSTR, swing1LimitDegrees=5, swing2LimitDegrees=20, twistLimitDegrees=2}
 {bone1Name=dyn_jacket_r_bk_02, bone2Name=dyn_jacket_r_bk_01, swing1LimitDegrees=5, swing2LimitDegrees=20, twistLimitDegrees=2}
 {bone1Name=dyn_jacket_l_bk_01, bone2Name=dyn_jacket_l_bk_MSTR, swing1LimitDegrees=5, swing2LimitDegrees=20, twistLimitDegrees=2}
 {bone1Name=dyn_jacket_l_bk_02, bone2Name=dyn_jacket_l_bk_01, swing1LimitDegrees=5, swing2LimitDegrees=20, twistLimitDegrees=2}
 {bone1Name=dyn_jacket_l_fr_01, bone2Name=dyn_jacket_l_fr_MSTR, swing1LimitDegrees=5, swing2LimitDegrees=30, twistLimitDegrees=5}
 {bone1Name=dyn_jacket_l_fr_02, bone2Name=dyn_jacket_l_fr_01, swing1LimitDegrees=5, swing2LimitDegrees=60, twistLimitDegrees=5}
 {bone1Name=dyn_jacket_r_fr_01, bone2Name=dyn_jacket_r_fr_MSTR, swing1LimitDegrees=5, swing2LimitDegrees=30, twistLimitDegrees=5}
 {bone1Name=dyn_jacket_r_fr_02, bone2Name=dyn_jacket_r_fr_01, swing1LimitDegrees=5, swing2LimitDegrees=60, twistLimitDegrees=5}
 {bone1Name=dyn_jacket_l_02, bone2Name=dyn_jacket_l_01, swing1LimitDegrees=2, swing2LimitDegrees=60, twistLimitDegrees=2}
 {bone1Name=dyn_jacket_l_01, bone2Name=dyn_jacket_l_MSTR, swing1LimitDegrees=2, swing2LimitDegrees=30, twistLimitDegrees=2}
 {bone1Name=dyn_jacket_r_01, bone2Name=dyn_jacket_r_MSTR, swing1LimitDegrees=5, swing2LimitDegrees=30, twistLimitDegrees=2}
 {bone1Name=dyn_jacket_r_02, bone2Name=dyn_jacket_r_01, swing1LimitDegrees=5, swing2LimitDegrees=60, twistLimitDegrees=2}
 {
  bone1Name=dyn_jacket_r_bk_MSTR
  bone2Name=spine_01
  swing1LimitDegrees=45
  swing1Motion=Free
  swing2LimitDegrees=45
  swing2Motion=Free
  twistLimitDegrees=45
  twistMotion=Free
 }
 {
  bone1Name=dyn_jacket_l_bk_MSTR
  bone2Name=spine_01
  swing1LimitDegrees=45
  swing1Motion=Free
  swing2LimitDegrees=45
  swing2Motion=Free
  twistLimitDegrees=45
  twistMotion=Free
 }
 {
  bone1Name=dyn_jacket_bk_MSTR
  bone2Name=spine_01
  swing1LimitDegrees=45
  swing1Motion=Free
  swing2LimitDegrees=45
  swing2Motion=Free
  twistLimitDegrees=45
  twistMotion=Free
 }
 {
  bone1Name=dyn_jacket_r_MSTR
  bone2Name=spine_01
  swing1LimitDegrees=45
  swing1Motion=Free
  swing2LimitDegrees=45
  swing2Motion=Free
  twistLimitDegrees=45
  twistMotion=Free
 }
 {
  bone1Name=dyn_jacket_l_MSTR
  bone2Name=spine_01
  swing1LimitDegrees=45
  swing1Motion=Free
  swing2LimitDegrees=45
  swing2Motion=Free
  twistLimitDegrees=45
  twistMotion=Free
 }
 {
  bone1Name=dyn_jacket_l_fr_MSTR
  bone2Name=spine_01
  swing1LimitDegrees=45
  swing1Motion=Free
  swing2LimitDegrees=45
  swing2Motion=Free
  twistLimitDegrees=45
  twistMotion=Free
 }
 {
  bone1Name=dyn_jacket_r_fr_MSTR
  bone2Name=spine_01
  swing1LimitDegrees=45
  swing1Motion=Free
  swing2LimitDegrees=45
  swing2Motion=Free
  twistLimitDegrees=45
  twistMotion=Free
 }
 {
  bone1Name=thigh_twist_01_l
  bone2Name=thigh_l
  swing1LimitDegrees=45
  swing1Motion=Free
  swing2LimitDegrees=45
  swing2Motion=Free
  twistLimitDegrees=45
  twistMotion=Free
 }
 {bone1Name=dyn_jacket_bk_01, bone2Name=dyn_jacket_bk_MSTR, swing1LimitDegrees=5, swing2LimitDegrees=20, twistLimitDegrees=2}
 {bone1Name=dyn_jacket_bk_02, bone2Name=dyn_jacket_bk_01, swing1LimitDegrees=5, swing2LimitDegrees=20, twistLimitDegrees=2}
 {
  bone1Name=spine_01
  bone2Name=pelvis
  swing1LimitDegrees=45
  swing1Motion=Free
  swing2LimitDegrees=45
  swing2Motion=Free
  twistLimitDegrees=45
  twistMotion=Free
 }
 {
  bone1Name=thigh_twist_01_r
  bone2Name=pelvis
  swing1LimitDegrees=45
  swing1Motion=Free
  swing2LimitDegrees=45
  swing2Motion=Free
  twistLimitDegrees=45
  twistMotion=Free
 }
 {
  bone1Name=thigh_l
  bone2Name=pelvis
  swing1LimitDegrees=45
  swing1Motion=Free
  swing2LimitDegrees=45
  swing2Motion=Free
  twistLimitDegrees=45
  twistMotion=Free
 }
 {
  bone1Name=dyn_jacket_l_bk_02
  bone2Name=dyn_jacket_bk_02
  swing1LimitDegrees=45
  swing1Motion=Free
  swing2LimitDegrees=45
  swing2Motion=Free
  twistLimitDegrees=45
  twistMotion=Free
 }
 {
  bone1Name=dyn_compass
  bone2Name=pelvis
  swing1LimitDegrees=45
  swing1Motion=Free
  swing2LimitDegrees=45
  swing2Motion=Free
  twistLimitDegrees=45
  twistMotion=Free
 }
 {bone1Name=dyn_collar_bk_r, bone2Name=dyn_collar_bk_r_MSTR, swing1LimitDegrees=2, swing2LimitDegrees=10, twistLimitDegrees=2}
 {
  bone1Name=dyn_collar_bk_r_MSTR
  bone2Name=spine_01
  swing1LimitDegrees=45
  swing1Motion=Free
  swing2LimitDegrees=45
  swing2Motion=Free
  twistLimitDegrees=45
  twistMotion=Free
 }
 {
  bone1Name=dyn_collar_r_MSTR
  bone2Name=spine_01
  swing1LimitDegrees=45
  swing1Motion=Free
  swing2LimitDegrees=45
  swing2Motion=Free
  twistLimitDegrees=45
  twistMotion=Free
 }
 {bone1Name=dyn_collar_r, bone2Name=dyn_collar_r_MSTR, swing1LimitDegrees=2, swing2LimitDegrees=25, twistLimitDegrees=2}
 {
  bone1Name=dyn_collar_ft_r_MSTR
  bone2Name=spine_01
  swing1LimitDegrees=45
  swing1Motion=Free
  swing2LimitDegrees=45
  swing2Motion=Free
  twistLimitDegrees=45
  twistMotion=Free
 }
 {bone1Name=dyn_collar_ft_r, bone2Name=dyn_collar_ft_r_MSTR, swing1LimitDegrees=2, swing2LimitDegrees=15, twistLimitDegrees=2}
 {
  bone1Name=head
  bone2Name=neck_02
  swing1LimitDegrees=45
  swing1Motion=Free
  swing2LimitDegrees=45
  swing2Motion=Free
  twistLimitDegrees=45
  twistMotion=Free
 }
 {
  bone1Name=dyn_jacket_r_fr_02
  bone2Name=dyn_jacket_r_02
  swing1LimitDegrees=45
  swing1Motion=Free
  swing2LimitDegrees=15
  twistLimitDegrees=45
  twistMotion=Free
 }
 {
  bone1Name=dyn_jacket_r_02
  bone2Name=dyn_jacket_r_bk_02
  swing1LimitDegrees=45
  swing1Motion=Free
  swing2LimitDegrees=45
  swing2Motion=Free
  twistLimitDegrees=45
  twistMotion=Free
 }
 {
  bone1Name=upperarm_twist_01_r
  bone2Name=spine_01
  swing1LimitDegrees=45
  swing1Motion=Free
  swing2LimitDegrees=45
  swing2Motion=Free
  twistLimitDegrees=45
  twistMotion=Free
 }
 {
  bone1Name=dyn_jacket_l_02
  bone2Name=dyn_jacket_l_bk_02
  swing1LimitDegrees=45
  swing1Motion=Free
  swing2LimitDegrees=45
  swing2Motion=Free
  twistLimitDegrees=45
  twistMotion=Free
 }
 {
  bone1Name=neck_02
  bone2Name=spine_01
  swing1LimitDegrees=45
  swing1Motion=Free
  swing2LimitDegrees=45
  swing2Motion=Free
  twistLimitDegrees=45
  twistMotion=Free
 }
]
