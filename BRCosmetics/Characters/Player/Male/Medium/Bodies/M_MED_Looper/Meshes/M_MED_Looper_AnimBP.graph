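# Post-process style AnimBP for the M_MED_Looper player body, copying the parent mesh pose and adding secondary motion to jacket, collar, straps, compass and radar.
# Cached BasePose applies two Control Rigs, compass constraint/look-at, strap AnimDynamics and jump-driven radar offsets.
# Jacket bone offsets switch between sprint and thigh-rotation-driven sets, then feed compass AnimDynamics and jacket/collar RigidBody simulations.
# EventGraph computes thigh alphas and state-based gravity override; jacket deform rig branch is unconnected.
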AnimBlueprint /BRCosmetics/Characters/Player/Male/Medium/Bodies/M_MED_Looper/Meshes/M_MED_Looper_AnimBP
target_skeleton=&/M_MED_Looper_Skeleton
graphs [
 AnimGraph {nodes [
   n1 = Control Rig(Source=Use cached pose 'BasePose'()) {
    alphaBoolBlend {blendOption=Linear}
    controlRigAssetReference {blueprintRigClass=&/M_MED_Looper_Jacket_Deform_CtrlRig.M_MED_Looper_Jacket_Deform_CtrlRig_C}
   }
   Blend Poses by bool(BlendPose_1=@n1, bActiveValue=Get bRigidBodyEnabled()) {blendType=Linear}
   "Constraint - attach_backpack (P)"(Alpha=1, ConstraintWeights_0=1) {
    alphaBoolBlend {blendOption=Linear}
    boneToModify {boneName=attach_backpack}
    constraintSetup [{targetBone {boneName=spine_05}}]
    constraintWeights [1]
   }
   n2 = "Transform (Modify) Bone - Bone: dyn_jacket_l_fr_MSTR"(ComponentPose=Local To Component(LocalPose=Use cached pose 'BasePose'()), Alpha=1, Translation="0,0,0", Rotation="15,0,-70", Scale="1,1,1") {
    rotationMode=BMM_Additive
    rotationSpace=BCS_BoneSpace
    translationMode=BMM_Additive
    translationSpace=BCS_BoneSpace
    alphaBoolBlend {blendOption=Linear}
    boneToModify {boneName=dyn_jacket_l_fr_MSTR}
    rotation {pitch=15, roll=-70}
   }
   n3 = "Transform (Modify) Bone - Bone: dyn_jacket_l_MSTR"(ComponentPose=@n2, Alpha=1, Translation="0,-2,1", Rotation="7.261,6.461,-35", Scale="1,1,1") {
    rotationMode=BMM_Additive
    rotationSpace=BCS_BoneSpace
    translationMode=BMM_Additive
    translationSpace=BCS_BoneSpace
    alphaBoolBlend {blendOption=Linear}
    boneToModify {boneName=dyn_jacket_l_MSTR}
    rotation=<rot 7.261 6.461 -35>
    translation {y=-2, z=1}
   }
   n4 = "Transform (Modify) Bone - Bone: dyn_jacket_l_bk_MSTR"(ComponentPose=@n3, Alpha=1, Translation="0,0,1", Rotation="-0.000,0.000,-10.000", Scale="1,1,1") {
    rotationMode=BMM_Additive
    rotationSpace=BCS_BoneSpace
    translationMode=BMM_Additive
    translationSpace=BCS_BoneSpace
    alphaBoolBlend {blendOption=Linear}
    boneToModify {boneName=dyn_jacket_l_bk_MSTR}
    rotation=<rot -0.0001 0.0001 -10>
    translation {z=1}
   }
   n5 = "Transform (Modify) Bone - Bone: dyn_jacket_r_fr_MSTR"(ComponentPose=@n4, Alpha=1, Translation="0,0,0", Rotation="15,5,60", Scale="1,1,1") {
    rotationMode=BMM_Additive
    rotationSpace=BCS_BoneSpace
    translationMode=BMM_Additive
    translationSpace=BCS_BoneSpace
    alphaBoolBlend {blendOption=Linear}
    boneToModify {boneName=dyn_jacket_r_fr_MSTR}
    rotation=<rot 15 5 60>
   }
   n6 = "Transform (Modify) Bone - Bone: dyn_jacket_r_MSTR"(ComponentPose=@n5, Alpha=1, Translation="0,3,1", Rotation="15,0,35", Scale="1,1,1") {
    rotationMode=BMM_Additive
    rotationSpace=BCS_BoneSpace
    translationMode=BMM_Additive
    translationSpace=BCS_BoneSpace
    alphaBoolBlend {blendOption=Linear}
    boneToModify {boneName=dyn_jacket_r_MSTR}
    rotation {pitch=15, roll=35}
    translation {y=3, z=1}
   }
   n7 = "Transform (Modify) Bone - Bone: dyn_jacket_r_bk_MSTR"(ComponentPose=@n6, Alpha=1, Translation="0,0,1", Rotation="-0.000,-0.000,15.000", Scale="1,1,1") {
    rotationMode=BMM_Additive
    rotationSpace=BCS_BoneSpace
    translationMode=BMM_Additive
    translationSpace=BCS_BoneSpace
    alphaBoolBlend {blendOption=Linear}
    boneToModify {boneName=dyn_jacket_r_bk_MSTR}
    rotation=<rot 0 0 15>
    translation {z=1}
   }
   n8 = Get bIsCrouchSprinting()
   n9 = "Transform (Modify) Bone - Bone: dyn_jacket_bk_MSTR"!(ComponentPose=@n7, bAlphaBoolEnabled:bool=@n8, Translation="2,0,0", Rotation="0,0,0", Scale="1,1,1") -> (Pose) {
    alphaInputType=Bool
    rotationMode=BMM_Additive
    rotationSpace=BCS_BoneSpace
    translationMode=BMM_Additive
    translationSpace=BCS_BoneSpace
    alphaBoolBlend {blendInTime=0.5, blendOption=Linear}
    boneToModify {boneName=dyn_jacket_bk_MSTR}
    translation {x=2}
   }
   n10 = "Transform (Modify) Bone - Bone: dyn_jacket_r_bk_MSTR"!(ComponentPose=@n9, bAlphaBoolEnabled:bool=@n8, Translation="1,0,0", Rotation="0,0,0", Scale="1,1,1") -> (Pose) {
    alphaInputType=Bool
    rotationMode=BMM_Additive
    rotationSpace=BCS_BoneSpace
    translationMode=BMM_Additive
    translationSpace=BCS_BoneSpace
    alphaBoolBlend {blendInTime=0.5, blendOption=Linear}
    boneToModify {boneName=dyn_jacket_r_bk_MSTR}
    translation {x=1}
   }
   n11 = "Transform (Modify) Bone - Bone: dyn_jacket_l_bk_MSTR"!(ComponentPose=@n10, bAlphaBoolEnabled:bool=@n8, Translation="1,0,0", Rotation="0,0,0", Scale="1,1,1") -> (Pose) {
    alphaInputType=Bool
    rotationMode=BMM_Additive
    rotationSpace=BCS_BoneSpace
    translationMode=BMM_Additive
    translationSpace=BCS_BoneSpace
    alphaBoolBlend {blendInTime=0.5, blendOption=Linear}
    boneToModify {boneName=dyn_jacket_l_bk_MSTR}
    translation {x=1}
   }
   n12 = Component To Local(ComponentPose=@n11)
   n13 = Get thigh_rot_up_l() -> (thigh_rot_up_l:double)
   n14 = "Transform (Modify) Bone - Bone: dyn_jacket_l_fr_MSTR"(
    ComponentPose=Local To Component(LocalPose=Use cached pose 'BasePose'())
    Alpha=@n13
    Translation="0,0,0"
    Rotation="40,-30,-70"
    Scale="1,1,1"
   ) {
    rotationMode=BMM_Additive
    rotationSpace=BCS_BoneSpace
    translationMode=BMM_Additive
    translationSpace=BCS_BoneSpace
    alphaBoolBlend {blendOption=Linear}
    boneToModify {boneName=dyn_jacket_l_fr_MSTR}
    rotation=<rot 40 -30 -70>
   }
   n15 = "Transform (Modify) Bone - Bone: dyn_jacket_l_MSTR"(ComponentPose=@n14, Alpha=@n13, Translation="-3,-2,2", Rotation="15,15,-30", Scale="1,1,1") {
    rotationMode=BMM_Additive
    rotationSpace=BCS_BoneSpace
    translationMode=BMM_Additive
    translationSpace=BCS_BoneSpace
    alphaBoolBlend {blendOption=Linear}
    boneToModify {boneName=dyn_jacket_l_MSTR}
    rotation=<rot 15 15 -30>
    translation=<-3 -2 2>
   }
   n16 = Get thigh_rot_up_r() -> (thigh_rot_up_r:double)
   n17 = "Transform (Modify) Bone - Bone: dyn_jacket_r_fr_MSTR"(ComponentPose=@n15, Alpha=@n16, Translation="0,0,0", Rotation="30,15,40", Scale="1,1,1") {
    rotationMode=BMM_Additive
    rotationSpace=BCS_BoneSpace
    translationMode=BMM_Additive
    translationSpace=BCS_BoneSpace
    alphaBoolBlend {blendOption=Linear}
    boneToModify {boneName=dyn_jacket_r_fr_MSTR}
    rotation=<rot 30 15 40>
   }
   n18 = "Transform (Modify) Bone - Bone: dyn_jacket_r_MSTR"(ComponentPose=@n17, Alpha=@n16, Translation="-4,0,1", Rotation="30,-5,15", Scale="1,1,1") {
    rotationMode=BMM_Additive
    rotationSpace=BCS_BoneSpace
    translationMode=BMM_Additive
    translationSpace=BCS_BoneSpace
    alphaBoolBlend {blendOption=Linear}
    boneToModify {boneName=dyn_jacket_r_MSTR}
    rotation=<rot 30 -5 15>
    translation {x=-4, z=1}
   }
   n19 = "Transform (Modify) Bone - Bone: dyn_jacket_r_bk_MSTR"!(ComponentPose=@n18, Alpha:float=@n16, Translation="-1.5,0,0", Rotation="0,-5,0", Scale="1,1,1") -> (Pose) {
    rotationMode=BMM_Additive
    rotationSpace=BCS_BoneSpace
    translationMode=BMM_Additive
    translationSpace=BCS_BoneSpace
    alphaBoolBlend {blendOption=Linear}
    boneToModify {boneName=dyn_jacket_r_bk_MSTR}
    rotation {yaw=-5}
    translation {x=-1.5}
   }
   n20 = Component To Local(ComponentPose=@n19)
   n21 = Get bIsSprinting()
   n22 = Anim Dynamics - Bone: dyn_compass(
    ComponentPose=Local To Component(LocalPose=Blend Poses by bool(BlendPose_0=@n12, BlendPose_1=@n20, BlendTime_0=0.4, BlendTime_1=0.3, bActiveValue=@n21) {blendType=Linear})
    Alpha=1
    ExternalForce="0,0,0"
   ) {
    angularDampingOverride=0.999
    bOverrideAngularDamping=true
    bUseGravityOverride=true
    simulationSpace=RootRelative
    alphaBoolBlend {blendOption=Linear}
    boundBone {boneName=dyn_compass}
    componentAppliedLinearAccClamp=<3000 3000 3000>
    componentLinearAccScale=<0.1 0.1 0.1>
    componentLinearVelScale=<0.1 0.1 0.1>
    gravityOverride {z=-980}
    physicsBodyDefinitions [{
      boundBone {boneName=dyn_compass}
      boxExtents=<10 18 15>
      constraintSetup {angularLimitsMax=<0 0 20>, angularLimitsMin=<0 -30 -30>, angularTarget=<0 0 0>}
      localJointOffset=<4 0 0>
     }]
   }
   n23 = Get GravityOverride()
   n24 = RigidBody!(ComponentPose=@n22, Alpha, OverrideWorldGravity:Vector=@n23, ExternalForce) -> (Pose) {
    bClampLinearTranslationLimitToRefPose=true
    bOverrideWorldGravity=true
    overridePhysicsAsset=&/M_MED_Looper_Dyn_Jacket
    simulationSpace=BaseBoneSpace
    alphaBoolBlend {blendOption=Linear}
    baseBoneRef {boneName=spine_01}
    componentAppliedLinearAccClamp=<3000 1000 1000>
    componentLinearAccScale {x=0.1}
    componentLinearVelScale {x=0.2}
    overrideWorldGravity {z=-980}
    simSpaceSettings {maxAngularAcceleration=100, maxAngularVelocity=5, maxLinearAcceleration=500, maxLinearVelocity=100, velocityScaleZ=0, worldAlpha=1}
   }
   n25 = RigidBody(ComponentPose=@n24, OverrideWorldGravity=@n23) {
    bClampLinearTranslationLimitToRefPose=true
    bOverrideWorldGravity=true
    overridePhysicsAsset=&/M_MED_Looper_Dyn_Collar
    simulationSpace=BaseBoneSpace
    alphaBoolBlend {blendOption=Linear}
    baseBoneRef {boneName=spine_05}
    componentAppliedLinearAccClamp=<3000 3000 3000>
    componentLinearAccScale {x=0.2}
    componentLinearVelScale {x=0.2}
    overrideWorldGravity {z=-980}
    simSpaceSettings {maxAngularAcceleration=100, maxAngularVelocity=5, maxLinearAcceleration=500, maxLinearVelocity=100, velocityScaleZ=0, worldAlpha=1}
   }
   n26 = "Transform (Modify) Bone - Bone: dyn_jacket_r_bk_MSTR"!(ComponentPose=@n25, bAlphaBoolEnabled:bool=True, Translation="0,0,1", Rotation="3,0,0", Scale="1,1,1") -> (Pose) {
    alphaInputType=Bool
    rotationMode=BMM_Additive
    rotationSpace=BCS_BoneSpace
    translationMode=BMM_Additive
    translationSpace=BCS_BoneSpace
    alphaBoolBlend {blendOption=Linear}
    boneToModify {boneName=dyn_jacket_r_bk_MSTR}
    rotation {pitch=3}
    translation {z=1}
   }
   n27 = "Transform (Modify) Bone - Bone: dyn_jacket_bk_MSTR"(ComponentPose=@n26, bAlphaBoolEnabled=True, Translation="0,0,1", Rotation="3,0,0", Scale="1,1,1") {
    alphaInputType=Bool
    rotationMode=BMM_Additive
    rotationSpace=BCS_BoneSpace
    translationMode=BMM_Additive
    translationSpace=BCS_BoneSpace
    alphaBoolBlend {blendOption=Linear}
    boneToModify {boneName=dyn_jacket_bk_MSTR}
    rotation {pitch=3}
    translation {z=1}
   }
   n28 = "Transform (Modify) Bone - Bone: dyn_jacket_l_bk_MSTR"(ComponentPose=@n27, bAlphaBoolEnabled=True, Translation="0,0,1", Rotation="5,0,0", Scale="1,1,1") {
    alphaInputType=Bool
    rotationMode=BMM_Additive
    rotationSpace=BCS_BoneSpace
    translationMode=BMM_Additive
    translationSpace=BCS_BoneSpace
    alphaBoolBlend {blendOption=Linear}
    boneToModify {boneName=dyn_jacket_l_bk_MSTR}
    rotation {pitch=5}
    translation {z=1}
   }
   Output Pose(Result=Component To Local(ComponentPose=@n28)) {layerGroup=None}
   n29 = Control Rig!(
    Source=Copy Pose From Mesh() {bUseAttachedParent=true}
    Alpha
    isTargeting:bool=Get bIsTargeting()
    isSkyDiving:bool=Get bIsSkyDiving()
    isJumping:bool=Get bIsJumping()
    isFalling:bool=Get bIsFalling()
   ) -> (Pose) {
    alphaBoolBlend {blendOption=Linear}
    controlRigAssetReference {blueprintRigClass=&/M_MED_Looper_Control_Rig.M_MED_Looper_Control_Rig_C}
   }
   n30 = Control Rig!(Source=@n29, Alpha) -> (Pose) {
    alphaBoolBlend {blendOption=Linear}
    controlRigAssetReference {blueprintRigClass=&/M_MED_Looper_DeformRig_ControlRig.M_MED_Looper_DeformRig_ControlRig_C}
   }
   n31 = "Constraint - dyn_compass_MSTR (PPPPP)"!(
    ComponentPose=Local To Component(LocalPose=@n30)
    Alpha=1
    ConstraintWeights_0=0.694
    ConstraintWeights_1:float=0.0071
    ConstraintWeights_2:float=0.0018
    ConstraintWeights_3:float=0.1255
    ConstraintWeights_4:float=0.1716
   ) -> (Pose) {
    alphaBoolBlend {blendOption=Linear}
    boneToModify {boneName=dyn_compass_MSTR}
    constraintSetup [
     {targetBone {boneName=pelvis}}
     {targetBone {boneName=thigh_twist_01_l}}
     {targetBone {boneName=deform_groin_l}}
     {targetBone {boneName=thigh_twist_01_r}}
     {targetBone {boneName=deform_groin_r}}
    ]
    constraintWeights [1,1,1,1,1]
   }
   n32 = Look At - Bone: dyn_compass_MSTR(ComponentPose=@n31, Alpha=1) {
    alphaBoolBlend {blendOption=Linear}
    boneToModify {boneName=dyn_compass_MSTR}
    lookAtLocation {x=0}
    lookAtTarget {boneReference {boneName=dyn_compass_TGT}}
    lookAt_Axis {axis=<1 0>}
   }
   n33 = "Transform (Modify) Bone - Bone: dyn_compass_MSTR"(ComponentPose=@n32, Alpha=Get thigh_rot_up_r(), Translation="0,0,4", Rotation="-15,0,0", Scale="1,1,1") {
    rotationMode=BMM_Additive
    rotationSpace=BCS_BoneSpace
    translationMode=BMM_Additive
    translationSpace=BCS_BoneSpace
    alphaBoolBlend {blendOption=Linear}
    boneToModify {boneName=dyn_compass_MSTR}
    rotation {pitch=-15}
    translation {z=4}
   }
   n34 = Anim Dynamics - Bone: dyn_straps_01(ComponentPose=@n33, Alpha=1, ExternalForce="0,0,0") {
    angularDampingOverride=0.9999
    bOverrideAngularDamping=true
    bUseGravityOverride=true
    simulationSpace=RootRelative
    alphaBoolBlend {blendOption=Linear}
    boundBone {boneName=dyn_straps_01}
    componentAppliedLinearAccClamp=<3000 3000 3000>
    componentLinearAccScale=<0.5 0.5 0.5>
    componentLinearVelScale=<0.3 0.3 0.3>
    gravityOverride {z=-980}
    physicsBodyDefinitions [{
      boundBone {boneName=dyn_straps_01}
      boxExtents=<10 10 15>
      constraintSetup {angularLimitsMax=<0 45 60>, angularLimitsMin=<0 -20 -60>, angularTarget=<0 0 0>}
      localJointOffset=<5 -1 0>
     }]
    relativeSpaceBone {boneName=calf_l}
   }
   n35 = Anim Dynamics - Bone: dyn_straps_02(ComponentPose=@n34, Alpha=1, ExternalForce="0,0,0") {
    angularDampingOverride=0.9999
    bOverrideAngularDamping=true
    bUseGravityOverride=true
    simulationSpace=RootRelative
    alphaBoolBlend {blendOption=Linear}
    boundBone {boneName=dyn_straps_02}
    componentAppliedLinearAccClamp=<3000 3000 3000>
    componentLinearAccScale=<0.5 0.5 0.5>
    componentLinearVelScale=<0.3 0.3 0.3>
    gravityOverride {z=-980}
    physicsBodyDefinitions [{
      boundBone {boneName=dyn_straps_02}
      boxExtents=<10 10 15>
      constraintSetup {angularLimitsMax=<0 20 60>, angularLimitsMin=<0 -45 -40>, angularTarget=<0 0 0>}
      localJointOffset=<4 -1 0>
     }]
    relativeSpaceBone {boneName=calf_l}
   }
   n36 = Get bIsJumping()
   n37 = "Transform (Modify) Bone - Bone: dyn_radar_01"!(ComponentPose=@n35, bAlphaBoolEnabled:bool=@n36, Translation="0,0,0", Rotation="-20,0,0", Scale="1,1,1") -> (Pose) {
    alphaInputType=Bool
    rotationMode=BMM_Additive
    rotationSpace=BCS_BoneSpace
    alphaBoolBlend {blendInTime=0.3, blendOption=Linear, blendOutTime=0.3}
    boneToModify {boneName=dyn_radar_01}
    rotation {pitch=-20}
   }
   n38 = "Transform (Modify) Bone - Bone: dyn_radar_02"!(ComponentPose=@n37, bAlphaBoolEnabled:bool=@n36, Translation="0,0,0", Rotation="45,0,0", Scale="1,1,1") -> (Pose) {
    alphaInputType=Bool
    rotationMode=BMM_Additive
    rotationSpace=BCS_BoneSpace
    alphaBoolBlend {blendInTime=0.3, blendOption=Linear, blendOutTime=0.3}
    boneToModify {boneName=dyn_radar_02}
    rotation {pitch=45}
   }
   n39 = "Transform (Modify) Bone - Bone: dyn_radar_handle"!(ComponentPose=@n38, bAlphaBoolEnabled:bool=@n36, Translation="0,0,0", Rotation="-5,0,0", Scale="1,1,1") -> (Pose) {
    alphaInputType=Bool
    rotationMode=BMM_Additive
    rotationSpace=BCS_BoneSpace
    alphaBoolBlend {blendInTime=0.3, blendOption=Linear, blendOutTime=0.3}
    boneToModify {boneName=dyn_radar_handle}
    rotation {pitch=-5}
   }
   n40 = "Transform (Modify) Bone - Bone: dyn_radar_03"!(ComponentPose=@n39, bAlphaBoolEnabled:bool=@n36, Translation="0,0,0", Rotation="-35,0,0", Scale="1,1,1") -> (Pose) {
    alphaInputType=Bool
    rotationMode=BMM_Additive
    rotationSpace=BCS_BoneSpace
    alphaBoolBlend {blendInTime=0.3, blendOption=Linear, blendOutTime=0.3}
    boneToModify {boneName=dyn_radar_03}
    rotation {pitch=-35}
   }
   n41 = "Transform (Modify) Bone - Bone: dyn_radar_04"!(ComponentPose=@n40, bAlphaBoolEnabled:bool=@n36, Translation="0,0,0", Rotation="15,0,0", Scale="1,1,1") -> (Pose) {
    alphaInputType=Bool
    rotationMode=BMM_Additive
    rotationSpace=BCS_BoneSpace
    alphaBoolBlend {blendInTime=0.3, blendOption=Linear, blendOutTime=0.3}
    boneToModify {boneName=dyn_radar_04}
    rotation {pitch=15}
   }
   Save cached pose 'BasePose'(Pose=Component To Local(ComponentPose=@n41)) {cacheName=BasePose}
  ]}
 EventGraph {nodes [
   Event BlueprintUpdateAnimation()
   > n1 = Sequence()
   n2 = Get SkelMesh()
   > Set thigh_rot_up_l(thigh_rot_up_l:double=GetAlphaFromReferenceTransformDelta(SkeletalMeshComponent=@n2, BoneName=thigh_l, BaseName=pelvis, DeltaComponentType=QuaternionTwist, InputRange=(X=0,Y=100), TwistAxis="0,0,1")) -> (Output_Get:double)
   > Set thigh_rot_up_r(thigh_rot_up_r:double=GetAlphaFromReferenceTransformDelta(SkeletalMeshComponent=@n2, BoneName=thigh_r, BaseName=pelvis, DeltaComponentType=QuaternionTwist, InputRange=(X=0,Y=100), TwistAxis="0,0,1")) -> (Output_Get:double)
   n3 = Gravity Override State Function DEPRECATED(
    execute=@n1.then_1
    CharacterPartAnimInstance=Self-Reference()
    GameTimeInSeconds=GetGameTimeInSeconds()
    GravityOverrideParams=Get Gravity Override Params()
   )
   > Set GravityOverride(GravityOverride=@n3)
   Event BlueprintInitializeAnimation()
   n4 = GetOwnerSkeletalMeshComponent()
   > Is Valid(InputObject=@n4)
   > Set SkelMesh(SkelMesh=@n4)
   FInterpEaseInOut!(A:double=0, B:double=0, Alpha:double=0, Exponent:double=0) -> (ReturnValue:double)
   Ease!(Function:EEasingFunc Enum=Linear, Alpha:double, A:Wildcard, B:Wildcard) -> (Result:Wildcard) {is=title K2Node_EaseFunction}
  ]}
]
variables [
 Gravity Override Params {
  type=GravityOverrideParamsStruct
  default {
   skydive_FloatingDown {gravityOverride=<0 -700 -100>, windAmplitude=<200 -500 200>, windOffset=<0 -600 100>}
   skydive_GoingUp {gravityOverride=<0 -300 500>, windAmplitude=<200 -500 300>}
   skydive_Parachuting {gravityOverride=<0 -500 -980>, windAmplitude=<200 -150 400>, windFrequency=<1.5 4 6>}
  }
 }
 GravityOverride {type=Vector}
 SkelMesh {type=SkeletalMeshComponent}
 thigh_rot_up_l {type=number}
 thigh_rot_up_r {type=number}
]
